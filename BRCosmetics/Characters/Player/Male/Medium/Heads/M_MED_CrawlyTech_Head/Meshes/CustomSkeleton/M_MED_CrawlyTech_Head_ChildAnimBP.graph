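# Child AnimBP for the CrawlyTech male head that procedurally twitches its left and right pincer base bones, switched off while an emote plays.
# The update event drives a clamped, smoothed sine whose speed, along with the squash or yaw target, depends on a Counter advanced after delays with random-based resets.
# Interpolated transforms feed additive bone-space rotation and scale on both bones.
# Pin rotations are computed but unused by the AnimGraph, and the Stage 3 composite is unconnected.
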AnimBlueprint /BRCosmetics/Characters/Player/Male/Medium/Heads/M_MED_CrawlyTech_Head/Meshes/CustomSkeleton/M_MED_CrawlyTech_Head_ChildAnimBP
target_skeleton=&/M_MED_CrawlyTech_Head_Skeleton
graphs [
 AnimGraph {nodes [
   n1 = Get!() -> (lip_trans_L_Location:Vector, lip_trans_L_Rotation:Rotator, lip_trans_L_Scale:Vector)
   n2 = NOT Boolean(A=Get bIsPlayingEmote())
   n3 = "Transform (Modify) Bone - Bone: L_base_pincer"!(
    ComponentPose=Local To Component(LocalPose=Input Pose())
    bAlphaBoolEnabled:bool=@n2
    Translation="0,0,0"
    Rotation=@n1.lip_trans_L_Rotation
    Scale=@n1.lip_trans_L_Scale
   ) -> (Pose) {
    alphaInputType=Bool
    rotationMode=BMM_Additive
    rotationSpace=BCS_BoneSpace
    scaleMode=BMM_Additive
    scaleSpace=BCS_BoneSpace
    alphaBoolBlend {blendOption=Linear}
    boneToModify {boneName=L_base_pincer}
   }
   n4 = Get!() -> (lip_trans_R_Location:Vector, lip_trans_R_Rotation:Rotator, lip_trans_R_Scale:Vector)
   n5 = "Transform (Modify) Bone - Bone: R_base_pincer"!(ComponentPose=@n3, bAlphaBoolEnabled:bool=@n2, Translation="0,0,0", Rotation=@n4.lip_trans_R_Rotation, Scale=@n4.lip_trans_R_Scale) -> (Pose) {
    alphaInputType=Bool
    rotationMode=BMM_Additive
    rotationSpace=BCS_BoneSpace
    scaleMode=BMM_Additive
    scaleSpace=BCS_BoneSpace
    alphaBoolBlend {blendOption=Linear}
    boneToModify {boneName=R_base_pincer}
   }
   Output Pose(Result=Component To Local(ComponentPose=@n5))
  ]}
 EventGraph {nodes [
   n1 = Event BlueprintUpdateAnimation()
   > Set Delta Time X(Delta Time X=@n1.DeltaTimeX)
   > n2 = Sequence!(execute) -> (then_0, then_1, then_2:Exec, then_3:Exec)
   n3 = "Sin (Radians)"(A=float * float(A=GetGameTimeInSeconds(), B=Get speedMultiplier() -> (speedMultiplier:double)) {is=title K2Node_PromotableOperator})
   Set SIN FLoat(
    execute=@n2.then_1
    SIN FLoat:double=FInterpTo(Current=Get SIN FLoat() -> (SIN FLoat:double), Target="Clamp (Float)"(Value=@n3, Min=0, Max=1), DeltaTime=Get Delta Time X(), InterpSpeed=35)
   ) -> (Output_Get:double)
   n4 = Get Counter() -> (Counter:int=0)
   K2Node_Composite_2 = Stage 1 Animation!() -> (ReturnValue:Transform) {is=title K2Node_Composite}
   K2Node_Composite_1 = Stage 2 Animation!() -> (ReturnValue:Transform, ReturnValue2:Transform) {is=title K2Node_Composite}
   Set lip_trans_L(
    execute=@n2.then_2
    lip_trans_L:Transform=TInterpTo!(
     Current:"Transform (by ref)"=Get lip_trans_L() -> (lip_trans_L:Transform)
     Target:"Transform (by ref)"=Select!(
      Option 0:Transform=@K2Node_Composite_2
      Option 1:Transform=@K2Node_Composite_1
      Option 2:Transform
      Option 3:Transform=@K2Node_Composite_2
      Option 4:Transform
      Option 5:Transform=@K2Node_Composite_1
      Index:int=@n4
     ) -> (ReturnValue:Transform)
     DeltaTime:float=Get Delta Time X()
     InterpSpeed:float=Get interpSpeed() -> (interpSpeed:double)
    ) -> (ReturnValue:Transform)
   ) -> (Output_Get:Transform)
   n5 = Get Counter() -> (Counter=0)
   > Set lip_trans_R(lip_trans_R:Transform=TInterpTo(
     Current=Get lip_trans_R() -> (lip_trans_R:Transform)
     Target=Select(
      Option 0=@K2Node_Composite_2
      Option 1=@K2Node_Composite_1.ReturnValue2
      Option 3=@K2Node_Composite_2
      Option 5=@K2Node_Composite_1.ReturnValue2
      Index=@n5
     )
     DeltaTime=Get Delta Time X()
     InterpSpeed=Get interpSpeed()
    )) -> (Output_Get:Transform)
   Set speedMultiplier(
    execute=@n2.then_0
    speedMultiplier:double=Select!(Option 0:double=36, Option 1:double=46, Option 2:double=36, Option 3:double=36, Index=Get Counter() -> (Counter=0)) -> (ReturnValue:double)
   ) -> (Output_Get:double)
   n6 = Get Counter() -> (Counter=0)
   > n7 = Branch(Condition="integer < integer"(A=RandomInteger!(Max:int=@n6) -> (ReturnValue:int=0), B=6))
   > Delay(Duration=0.85)
   > Set Counter(Counter:int=int + int(A=Get Counter() -> (Counter=0))) -> (Output_Get:int=0)
   Set Counter(execute=@n7.else, Counter=0) -> (Output_Get=0)
   Reroute Node()
   n8 = Get Counter() -> (Counter=0)
   n9 = Get SIN FLoat()
   n10 = float * float(
    A=Remap Float(Input="Clamp (Float)"(Value=@n9, Min=0, Max=1), OldRange=(X=0,Y=1), NewRange=(X=0.5,Y=1), Clamp=true) -> (Output=~)
    B="To Float (Boolean)"(InBool="Equal (Integer)"(A=@n8, B=3))
   )
   > Set Pin01_rot_L(Pin01_rot_L:Rotator=RInterpTo!(
     Current:Rotator=Get Pin01_rot_L() -> (Pin01_rot_L:Rotator)
     Target_Roll:float=0
     Target_Pitch:float=float * float(A=@n10, B=-90)
     Target_Yaw:float=0
     DeltaTime:float=Get Delta Time X()
     InterpSpeed:float=Get interpSpeed()
    ) -> (ReturnValue:Rotator)) -> (Output_Get:Rotator)
   > Set Pin01_rot_R(Pin01_rot_R:Rotator=RInterpTo(
     Current=Get Pin01_rot_R() -> (Pin01_rot_R:Rotator)
     Target_Roll=0
     Target_Pitch=float * float(A=@n10, B=90)
     Target_Yaw=0
     DeltaTime=Get Delta Time X()
     InterpSpeed=Get interpSpeed()
    )) -> (Output_Get:Rotator)
   K2Node_Composite_0 = Stage 3 Animation!() -> (ReturnValue:Transform, ReturnValue2:Transform) {is=title K2Node_Composite}
   Reroute Node(InputPin:Transform=Reroute Node()) -> (OutputPin:Transform)
  ]}
 Stage 3 Animation {
  path=EventGraph.K2Node_Composite_0.Stage 3 Animation
  nodes [
   Inputs!()
   n1 = Remap Float(Input="Clamp (Float)"(Value=Get SIN FLoat(), Min=0, Max=1), OldRange=(X=0,Y=1), NewRange=(X=0.5,Y=1), Clamp=true) -> (Output=~)
   n2 = Remap Float(Input=@n1, OldRange=(X=0,Y=1), NewRange=(X=1,Y=0.8), Clamp=true) -> (Output=~)
   n3 = Remap Float(Input=@n2, OldRange=(X=1,Y=0.8), NewRange=(X=1,Y=1.25), Clamp=true) -> (Output=~)
   Outputs!(
    ReturnValue:Transform=MakeTransform!(
     Location
     Rotation_Roll:float=0
     Rotation_Pitch:float=float * float(A=@n1, B=45)
     Rotation_Yaw:float=0
     Scale_X:double=@n2
     Scale_Y:double=@n3
     Scale_Z:double=@n3
    ) -> (ReturnValue)
    ReturnValue2:Transform=MakeTransform(Rotation_Roll=0, Rotation_Pitch=float * float(A=@n1, B=-45), Rotation_Yaw=0, Scale_X=@n2, Scale_Y=@n3, Scale_Z=@n3)
   )
  ]
 }
 Stage 2 Animation {
  path=EventGraph.K2Node_Composite_1.Stage 2 Animation
  nodes [
   Inputs()
   n1 = float * float(A=Get SIN FLoat(), B=-16)
   Outputs(
    ReturnValue=MakeTransform!(Location, Rotation_Roll=0, Rotation_Pitch=0, Rotation_Yaw=@n1, Scale:Vector) -> (ReturnValue)
    ReturnValue2=MakeTransform(Rotation_Roll=0, Rotation_Pitch=0, Rotation_Yaw=@n1)
   )
  ]
 }
 Stage 1 Animation {
  path=EventGraph.K2Node_Composite_2.Stage 1 Animation
  nodes [
   Inputs()
   n1 = Remap Float(Input="Absolute (Float)"(A=Get SIN FLoat()), OldRange=(X=0,Y=1), NewRange=(X=0.9,Y=1.1), Clamp=true) -> (Output=~)
   n2 = Remap Float(Input=@n1, OldRange=(X=1.1,Y=0.9), NewRange=(X=0.9,Y=1.1), Clamp=true) -> (Output=~)
   Outputs!(ReturnValue=MakeTransform!(Location, Rotation:Rotator="0, 0, 0", Scale_X:double=@n1, Scale_Y:double=@n2, Scale_Z:double=@n2) -> (ReturnValue))
  ]
 }
]
variables [
 animationEnabled {type=boolean, default=true}
 Counter {type=integer}
 Delta Time X {type=number}
 interpSpeed {type=number, default=20}
 lip_trans_L {type=Transform, default {location=<0 0 0>, rotation=<rot 0 0 0>, scale=<1 1 1>}}
 lip_trans_R {type=Transform, default {location=<0 0 0>, rotation=<rot 0 0 0>, scale=<1 1 1>}}
 Pin01_rot_L {type=Rotator}
 Pin01_rot_R {type=Rotator}
 SIN FLoat {type=number}
 speedMultiplier {type=number, default=36}
]
